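# Material function in the Chromakeying library that manually removes key color spill from an image.
# An inverted, saturated ramp of Raw Comparison, shaped by Despill Minimum and Divisor, forms the spill mask.
# The image times the normalized keyed color, mask and Despill Amount is subtracted from the image and saturated.
# A fully desaturated copy tinted by Sky Color and scaled by Sky Intensity and mask is added as fake sky bounce fill.
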MaterialFunction /Engine/Functions/Engine_MaterialFunctions01/Chromakeying/DespillManual
graph {
 properties {description=Manual despill function, libraryCategoriesText [Chromakeying]}
 nodes [
  n1 = FunctionInput() {bUsePreviewValueAsDefault=true, description=The input image to apply despill on, inputName=Image Color, previewValue=<1 1 1 1>}
  n2 = Multiply(
   A=Desaturation(None=@n1, Fraction=Constant() {r=1})
   B=FunctionInput() {bUsePreviewValueAsDefault=true, inputName=Sky Color, sortPriority=5, previewValue=<1 1 1 1>}
  )
  n3 = FunctionInput(Preview=Constant() {r=1}) {
   bUsePreviewValueAsDefault=true
   description=Fake sky bounce fill color
   inputName=Sky Intensity
   inputType=FunctionInput_Scalar
   sortPriority=6
  }
  n4 = Subtract(
   A=FunctionInput() {description=Image comparison with key color, inputName=Raw Comparison, sortPriority=7}
   B=FunctionInput(Preview=Constant() {r=0.3}) {bUsePreviewValueAsDefault=true, inputName=Despill Minimum, inputType=FunctionInput_Scalar, sortPriority=2}
  )
  n5 = FunctionInput(Preview=Constant() {r=0.3}) {bUsePreviewValueAsDefault=true, description=Soften despill effect, inputName=Despill Divisor, inputType=FunctionInput_Scalar, sortPriority=3}
  n6 = OneMinus(None=Saturate(None=Divide(A=@n4, B=@n5)))
  n7 = Multiply(A=@n3, B=@n6)
  n8 = Multiply(A=@n1, B=Normalize(VectorInput=FunctionInput() {inputName=KeyedColor, sortPriority=1}))
  n9 = Constant() {r=0.3}
  n10 = Subtract(
   A=@n1
   B=Multiply(
    A=Multiply(A=@n8, B=@n6)
    B=FunctionInput(Preview=@n9) {bUsePreviewValueAsDefault=true, inputName=Despill Amount, inputType=FunctionInput_Scalar, sortPriority=4}
   )
  ) {constB=0}
  FunctionOutput(None=Add(A=Multiply(A=@n2, B=@n7), B=Saturate(None=@n10)))
 ]
}
