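# ColorCorrection is a material function in the Image Adjustment library that color-grades an input color.
# It applies three-way Lift weighted by the inverted color, multiplies by Gain, then raises the result to a Gamma-derived exponent.
# An SCurve function adds Contrast, and Desaturation follows, driven by Saturation subtracted from a constant.
# The HueShift function then rotates hue, and a Brightness multiply produces the output.
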MaterialFunction /Engine/Functions/Engine_MaterialFunctions01/ImageAdjustment/ColorCorrection
graph {
 properties {description=Color correction, libraryCategoriesText [Image Adjustment]}
 nodes [
  n1 = FunctionInput() {inputName=Hue Shift, inputType=FunctionInput_Scalar}
  n2 = FunctionInput() {description=3 Way Lift Control, inputName=Lift}
  n3 = FunctionInput() {description=Input color to apply correction, inputName=Color}
  n4 = OneMinus(None=@n3)
  n5 = Multiply(A=Add(A=Multiply(A=@n2, B=@n4), B=@n3), B=FunctionInput() {description=3 Way Color Gain Control, inputName=Gain})
  n6 = Divide(B=FunctionInput() {description=3 Way Gamma Control, inputName=Gamma})
  n7 = Desaturation(
   None=MaterialFunctionCall(In=Power(Base=@n5, Exp=@n6), Power=FunctionInput() {inputName=Contrast, inputType=FunctionInput_Scalar}) {materialFunction=&SCurve}.Result
   Fraction=Subtract(A=Constant(), B=FunctionInput() {inputName=Saturation, inputType=FunctionInput_Scalar})
  )
  FunctionOutput(None=Multiply(
    A=MaterialFunctionCall(Hue Shift Percentage=@n1, Texture=@n7) {materialFunction=&HueShift}.Result
    B=FunctionInput() {inputName=Brightness, inputType=FunctionInput_Scalar}
   ))
 ]
}
