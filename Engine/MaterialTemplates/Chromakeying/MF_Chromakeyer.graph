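# Template material function for parameterized chromakeying, turning an input image into an Emissive Color and an Opacity matte.
# Chroma_Key_Alpha pulls the matte, UV cropping trims it, and pixels below IntensityFloor are forced opaque.
# A static switch picks automatic or manual despill, then blends in color correction and an alpha preview.
# Opacity is the minimum of that matte and the floored garbage matte texture alpha.
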MaterialFunction /Engine/MaterialTemplates/Chromakeying/MF_Chromakeyer
graph {
 properties {
  bExposeToLibrary=true
  description=Parameterized template of a Chromakeying material with 2d garbage matting
  libraryCategoriesText [Chromakeying]
 }
 nodes [
  n1 = FunctionInput() {
   bUsePreviewValueAsDefault=true
   description=Image you want to apply chromakeying to
   inputName=Input Color
   sortPriority=1
   previewValue=<1 1 1 1>
  }
  n2 = VectorParameter() {
   desc=Despill Color Selection
   group=Keyer_03 - Despill
   parameterName=Despill Color
   sortPriority=1
   defaultValue=<rgba 0.0704 0.1095 0.0666 1>
  }
  n3 = ScalarParameter() {defaultValue=1, desc=Despill Amount, group=Keyer_03 - Despill, parameterName=Despill Amount, sliderMax=2, sliderMin=-2, sortPriority=6}
  n4 = VectorParameter() {group=Keyer_03 - Despill, parameterName=Sky, sortPriority=4, defaultValue=<rgba 0.2124 0.2386 0.3372 1>}
  n5 = ScalarParameter() {defaultValue=1, desc=Fake sky bounce fill color, group=Keyer_03 - Despill, parameterName=Sky Intensity, sortPriority=3}
  n6 = ScalarParameter() {
   desc="Luminance value used to pull the matte.  Lower until the backing is cleared.  Use your alpha preview to check"
   group=Keyer_02 - Alpha Control
   parameterName=LumaMask
   sortPriority=1
  }
  n7 = VectorParameter() {desc=Key Color Selection, group=Keyer_01 - Color Sample, parameterName=Key Color, defaultValue=<rgba 0.0704 0.1095 0.0666 1>}
  n8 = ScalarParameter() {desc=Raise values to help clear the backing., group=Keyer_02 - Alpha Control, parameterName=Alpha Min, sortPriority=2}
  n9 = ScalarParameter() {
   defaultValue=0.15
   desc=Divides Alpha and softens edges helps control fine hair detail
   group=Keyer_02 - Alpha Control
   parameterName=Alpha Divisor
   sortPriority=3
  }
  n10 = MaterialFunctionCall!(
   LumaMask=@n6
   Image Color=@n1
   Chroma Color=@n7.RGB
   Alpha Cutoff Min=@n8
   Alpha Cutoff Max=Add(A=@n8, B=@n9)
   Alpha Exponent=Constant() {r=1}
   Despill Cuttoff Max
   Despill Exponent
  ) {materialFunction=/Engine/Functions/Engine_MaterialFunctions01/Opacity/Chroma_Key_Alpha}
  n11 = StaticSwitchParameter(
   True=MaterialFunctionCall!(Image Color=@n1) {materialFunction=/Engine/Functions/Engine_MaterialFunctions01/Chromakeying/DespillByAvg}.Result
   False=MaterialFunctionCall!(
    Image Color=@n1
    KeyedColor=@n2.RGB
    Despill Minimum=ScalarParameter() {defaultValue=0.3, desc=Despill Minimum value, group=Keyer_03 - Despill, parameterName=Despill Min, sortPriority=2}
    Despill Divisor=ScalarParameter() {defaultValue=0.35, desc=Despill Soften, group=Keyer_03 - Despill, parameterName=Despill Divisor, sortPriority=5}
    Despill Amount=@n3
    Sky Color=@n4.RGB
    Sky Intensity=@n5
    Raw Comparison=@n10.Raw Comparison
   ) {materialFunction=/Engine/Functions/Engine_MaterialFunctions01/Chromakeying/DespillManual}.Result
  ) {defaultValue=true, desc=Enable automatic despill method, group=Keyer_03 - Despill, parameterName=Despill Automatic}
  n12 = ScalarParameter() {desc=Global Color Correction Amount, group=Keyer_07 - Vizualisation, parameterName=Color Correction Amount, sliderMax=1}
  n13 = VectorParameter() {desc=3 Way Color Gain Control, group=Keyer_05 - Color Correction, parameterName=Gain, sortPriority=3, defaultValue=<rgba 1 1 1 0>}
  n14 = VectorParameter() {desc=3 Way Gamma Control, group=Keyer_05 - Color Correction, parameterName=Gamma, sortPriority=2, defaultValue=<rgba 1 1 1 0>}
  n15 = MaterialFunctionCall!(
   Gain=@n13.RGB
   Gamma=@n14.RGB
   Lift=VectorParameter() {desc=3 Way Lift Control, group=Keyer_05 - Color Correction, parameterName=Lift, sortPriority=1}.RGB
   Contrast=ScalarParameter() {defaultValue=1, group=Keyer_05 - Color Correction, parameterName=Contrast, sortPriority=81}
   Saturation=ScalarParameter() {group=Keyer_05 - Color Correction, parameterName=Saturation, sliderMax=2, sliderMin=-2, sortPriority=82}
   Hue Shift=ScalarParameter() {group=Keyer_05 - Color Correction, parameterName=Hue Shift, sliderMax=0.5, sliderMin=-0.5, sortPriority=83}
   Brightness=ScalarParameter() {defaultValue=1, group=Keyer_05 - Color Correction, parameterName=Brightness, sliderMax=8, sortPriority=80}
   Color=@n11
  ) {materialFunction=/Engine/Functions/Engine_MaterialFunctions01/ImageAdjustment/ColorCorrection}
  n16 = ScalarParameter() {desc=Minimum intensity a pixel must have to be keyable, group=Keyer_02 - Alpha Control, parameterName=IntensityFloor, sortPriority=4}
  n17 = ScalarParameter() {desc=Crop Right, group=Keyer_04 - Cropping, parameterName=Crop Right, sliderMax=1, sortPriority=1}
  n18 = ScalarParameter() {desc=Crop Bottom, group=Keyer_04 - Cropping, parameterName=Crop Bottom, sliderMax=1, sortPriority=4}
  n19 = ScalarParameter() {desc=Crop Top, group=Keyer_04 - Cropping, parameterName=Crop Top, sliderMax=1, sortPriority=2}
  n20 = ScalarParameter() {desc=Crop Left, group=Keyer_04 - Cropping, parameterName=Crop Left, sliderMax=1}
  n21 = Max(
   A=If(A=Desaturation(None=@n1), B=@n16, "A > B"=Constant(), "A < B"=Constant() {r=1})
   B=Multiply(
    A=@n10.Alpha
    B=MaterialFunctionCall!(Crop Right=@n17, Crop Bottom=@n18, Crop Top=@n19, Crop Left=@n20) {materialFunction=/Engine/Functions/Engine_MaterialFunctions01/Texturing/UVCropping}.Result
   )
  )
  FunctionOutput(None=LinearInterpolate(
    A=LinearInterpolate(A=@n11, B=@n15.Result, Alpha=@n12)
    B=@n21
    Alpha=ScalarParameter() {desc=Alpha Preview Amount, group=Keyer_07 - Vizualisation, parameterName=Alpha Preview, sliderMax=1, sortPriority=1}
   )) {outputName=Emissive Color}
  n22 = FunctionInput() {
   bUsePreviewValueAsDefault=true
   description=UV to sample garbage matte texture
   inputName=InputUV
   inputType=FunctionInput_Vector2
   sortPriority=1
   previewValue=<1 1 1 1>
  }
  n23 = TextureSampleParameter2D(UVs=@n22) {group=Keyer_06 - Garbage Matte Mask, parameterName=GarbageMatteTexture, texture=&Black}
  FunctionOutput(None=Min(A=@n21, B=Floor(None=@n23.A))) {outputName=Opacity, sortPriority=1}
 ]
}
