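# Material function for chroma keying that removes color spill by desaturating input colors near a chosen despill hue.
# Both the green screen input and the Despill Hue color convert to HSV, and the absolute difference of their hue channels is taken without wraparound.
# A reversed SmoothStep from Hue Range down to zero turns that difference into a mask, blending toward a copy desaturated by DeSpill Amount.
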MaterialFunction /Engine/Functions/Engine_MaterialFunctions01/Chromakeying/DespillByHue
graph {nodes [
  n1 = FunctionInput() {description=Green Screen Input, inputName=Texture, sortPriority=1}
  n2 = FunctionInput() {
   bUsePreviewValueAsDefault=true
   desc=DeSpill Amount
   description=DeSpill Amount
   inputName=DeSpill Amount
   inputType=FunctionInput_Scalar
   sortPriority=4
   previewValue=<1 0 0 1>
  }
  n3 = MaterialFunctionCall!(In=@n1) {materialFunction=&/RGBtoHSV}
  n4 = MaterialFunctionCall(In=FunctionInput() {inputName=Despill Hue, sortPriority=2, previewValue=<0 0.45 0 1>}) {materialFunction=&RGBtoHSV}
  n5 = Abs(None=Subtract(A=ComponentMask(None=@n3.Result) {r=true}, B=ComponentMask(None=@n4.Result) {r=true}))
  n6 = FunctionInput() {
   bUsePreviewValueAsDefault=true
   desc=Hue Range
   description=Hue Range
   inputName=Hue Range
   inputType=FunctionInput_Scalar
   sortPriority=3
   previewValue=<0.2 0 0 1>
  }
  n7 = Constant()
  FunctionOutput(None=LinearInterpolate(
    A=@n1
    B=Desaturation(None=@n1, Fraction=@n2)
    Alpha=MaterialFunctionCall(Alpha=@n5, Min=@n6, Max=@n7) {materialFunction=&SmoothStep}.Result
   ))
 ]}
