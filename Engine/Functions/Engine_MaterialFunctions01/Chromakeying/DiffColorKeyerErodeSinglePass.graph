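# Material function for chromakeying that removes a key color by color difference, with optional single-pass alpha erosion, outputting RGB and Alpha.
# DiffColorKeyer keys a Texture2D or external texture; the ErodeAlpha switch selects a custom HLSL version instead.
# That HLSL re-keys Hammersley-distributed neighbor samples, taking minimum alpha for positive kernel size or maximum for negative, assuming 1024-square textures.
# It calls keyer functions by name, relying on DiffColorKeyer's generated code.
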MaterialFunction /Engine/Functions/Engine_MaterialFunctions01/Chromakeying/DiffColorKeyerErodeSinglePass
graph {
 properties {bExposeToLibrary=true, description=Color difference based keyer, libraryCategoriesText [Chromakeying]}
 nodes [
  FunctionInput() {
   bUsePreviewValueAsDefault=true
   description=---------------------
   inputName=---------------------
   inputType=FunctionInput_StaticBool
   sortPriority=9
  }
  MaterialExpressionStaticSwitch_1 = StaticSwitch(
   True=FunctionInput() {bUsePreviewValueAsDefault=true, inputName=TextureExternal, inputType=FunctionInput_TextureExternal}
   False=FunctionInput(Preview=TextureObject()) {bUsePreviewValueAsDefault=true, inputName=Texture2D, inputType=FunctionInput_Texture2D}
   Value=FunctionInput(Preview=StaticBool()) {bUsePreviewValueAsDefault=true, inputName=IsExternalSource, inputType=FunctionInput_StaticBool}
  )
  MaterialExpressionTextureCoordinate_0 = TextureCoordinate()
  MaterialExpressionConstant2Vector_1 = Constant2Vector() {g=1024, r=1024}
  MaterialExpressionFunctionInput_2 = FunctionInput() {bUsePreviewValueAsDefault=true, inputName=KernalSize, inputType=FunctionInput_Scalar, sortPriority=11}
  MaterialExpressionFunctionInput_3 = FunctionInput() {bUsePreviewValueAsDefault=true, inputName=NumSamples, inputType=FunctionInput_Scalar, sortPriority=12, previewValue=<16 0 0 1>}
  MaterialExpressionTextureSample_0 = TextureSample(UVs=@MaterialExpressionTextureCoordinate_0, Tex=@MaterialExpressionStaticSwitch_1)
  MaterialExpressionFunctionInput_1 = FunctionInput() {bUsePreviewValueAsDefault=true, inputName=KeyColor, sortPriority=1, previewValue=<0 1 0 1>}
  n1 = FunctionInput() {bUsePreviewValueAsDefault=true, inputName=AlphaThreshold, inputType=FunctionInput_Scalar, sortPriority=2, previewValue=<1 0 0 1>}
  MaterialExpressionAppendVector_2 = AppendVector(A=@n1, B=FunctionInput() {bUsePreviewValueAsDefault=true, inputName=AlphaOffset, inputType=FunctionInput_Scalar, sortPriority=3})
  n2 = FunctionInput() {bUsePreviewValueAsDefault=true, inputName=WeightRed, inputType=FunctionInput_Scalar, sortPriority=4, previewValue=<0.5 0 0 1>}
  n3 = FunctionInput() {bUsePreviewValueAsDefault=true, inputName=WeightBlue, inputType=FunctionInput_Scalar, sortPriority=5, previewValue=<0.5 0 0 1>}
  MaterialExpressionAppendVector_3 = AppendVector(A=@n2, B=@n3)
  n4 = FunctionInput() {bUsePreviewValueAsDefault=true, inputName=ClipWhite, inputType=FunctionInput_Scalar, sortPriority=7, previewValue=<1 0 0 1>}
  MaterialExpressionAppendVector_4 = AppendVector(A=FunctionInput() {bUsePreviewValueAsDefault=true, inputName=ClipBlack, inputType=FunctionInput_Scalar, sortPriority=6}, B=@n4)
  MaterialExpressionFunctionInput_4 = FunctionInput() {bUsePreviewValueAsDefault=true, inputName=UnPremult, inputType=FunctionInput_Scalar, sortPriority=8}
  MaterialExpressionMaterialFunctionCall_0 = MaterialFunctionCall!(
   GreenScreen=@MaterialExpressionTextureSample_0.RGB
   KeyColor=@MaterialExpressionFunctionInput_1
   AlphaThresOffset=@MaterialExpressionAppendVector_2
   WeightsRB=@MaterialExpressionAppendVector_3
   ClipBW=@MaterialExpressionAppendVector_4
   UnPremult=@MaterialExpressionFunctionInput_4
  ) {materialFunction=&/DiffColorKeyer}
  n5 = Custom!(
   Tex=@MaterialExpressionStaticSwitch_1
   UVs=@MaterialExpressionTextureCoordinate_0
   TextureSize=@MaterialExpressionConstant2Vector_1
   KernelSize=@MaterialExpressionFunctionInput_2
   NumSamples=@MaterialExpressionFunctionInput_3
   DiffKeyer=@MaterialExpressionMaterialFunctionCall_0.RGBA
   GreenScreen=@MaterialExpressionTextureSample_0.RGB
   KeyColor=@MaterialExpressionFunctionInput_1
   AlphaThresOffset=@MaterialExpressionAppendVector_2
   WeightsRB=@MaterialExpressionAppendVector_3
   ClipBW=@MaterialExpressionAppendVector_4
   Unpremult=@MaterialExpressionFunctionInput_4
  ) {
   code="// inputs\r\n// Tex: GreenScreen Texture Object\r\n// UVs: Texture coordinates\r\n// KernelSize: filter radius\r\n// NumSamples: filter samples count\r\n// TextureSize: dimensions used to calculate the filter radius\r\n\r\n// DiffKeyer: Custom HLSL function for the ColorDiffKeyer\r\n// ... rest of parameters listed in ColorDiffKeyer\r\n\r\n// should cache samples for blur?\r\n\r\nfloat2 blurRadius = KernelSize / TextureSize;\r\nfloat4 finalColor = CustomExpression0(Parameters, GreenScreen, KeyColor, AlphaThresOffset, WeightsRB, ClipBW, Unpremult);\r\n\r\nfor (int i=0;i<NumSamples;++i)\r\n{\r\n\tfloat2 pdf = CosineSampleHemisphere(Hammersley(i, NumSamples, 0)).xy;\r\n\tfloat2 uvOffset = pdf * blurRadius;\r\n\tfloat2 uvShifted = UVs + uvOffset;\r\n\tfloat4 hitSample = Texture2DSample(Tex, TexSampler, saturate(uvShifted));\r\n\tfloat4 hitColor = GetDiffColorKeyer(hitSample, KeyColor, AlphaThresOffset, WeightsRB, ClipBW, Unpremult);\r\n\tif (finalColor.a > hitColor.a &&&& KernelSize > 0)\r\n\t{\r\n\t\tfinalColor.a = hitColor.a;\r\n\t}\r\n\tif (finalColor.a < hitColor.a &&&& KernelSize < 0)\r\n\t{\r\n\t\tfinalColor.a = hitColor.a;\r\n\t}\r\n}\r\n\r\nreturn finalColor;"
   description=ErodeAlphaDiffColor
   outputType=CMOT_Float4
   inputs [
    {
     inputName=Tex
     input {expression=&self:MaterialExpressionStaticSwitch_1}
    }
    {
     inputName=UVs
     input {expression=&self:MaterialExpressionTextureCoordinate_0}
    }
    {
     inputName=TextureSize
     input {expression=&self:MaterialExpressionConstant2Vector_1}
    }
    {
     inputName=KernelSize
     input {expression=&self:MaterialExpressionFunctionInput_2}
    }
    {
     inputName=NumSamples
     input {expression=&self:MaterialExpressionFunctionInput_3}
    }
    {
     inputName=DiffKeyer
     input {expression=&self:MaterialExpressionMaterialFunctionCall_0}
    }
    {
     inputName=GreenScreen
     input {
      expression=&self:MaterialExpressionTextureSample_0
      mask=1
      maskB=1
      maskG=1
      maskR=1
     }
    }
    {
     inputName=KeyColor
     input {expression=&self:MaterialExpressionFunctionInput_1}
    }
    {
     inputName=AlphaThresOffset
     input {expression=&self:MaterialExpressionAppendVector_2}
    }
    {
     inputName=WeightsRB
     input {expression=&self:MaterialExpressionAppendVector_3}
    }
    {
     inputName=ClipBW
     input {expression=&self:MaterialExpressionAppendVector_4}
    }
    {
     inputName=Unpremult
     input {expression=&self:MaterialExpressionFunctionInput_4}
    }
   ]
  }
  n6 = StaticSwitch(
   True=@n5
   False=@MaterialExpressionMaterialFunctionCall_0.RGBA
   Value=FunctionInput(Preview=StaticBool()) {bUsePreviewValueAsDefault=true, inputName=ErodeAlpha, inputType=FunctionInput_StaticBool, sortPriority=10}
  )
  FunctionOutput(None=ComponentMask(None=@n6) {b=true, g=true, r=true}) {outputName=RGB}
  FunctionOutput(None=ComponentMask(None=@n6) {a=true}) {outputName=Alpha}
 ]
}
