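# Unlit translucent chroma key material template that cuts a green screen out of external media or a 2D texture.
# A color-difference keyer function with optional erosion builds the matte, which a Custom rounded-rectangle distance node can fade at the edges.
# Keyed color passes through switchable despill, blending hue-plus-average and average-only methods, then optional color correction.
# Debug Alpha blends color to the raw matte and opacity to one for inspection.
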Material /Engine/MaterialTemplates/Chromakeying/M_ColorDifferenceKey
graph {
 properties {blendMode=BLEND_Translucent, pixelDepthOffsetMode=PDOM_Legacy, shadingModel=MSM_Unlit, usedShadingModels=Unlit}
 nodes [
  n1 = TextureObjectParameter() {group=01 Green Screen, parameterName=TextureMedia, samplerType=SAMPLERTYPE_External, texture=/Engine/EngineResources/DefaultMediaTexture}
  n2 = VectorParameter() {
   desc=Select a shade of green for keying
   group=01 Green Screen
   parameterName=Key Color
   sortPriority=1
   defaultValue=<rgba 0.0303 0.401 0 1>
  }
  n3 = ScalarParameter() {defaultValue=0.5, desc=Red Channel influence, group=01 Green Screen, parameterName=Red Weight, sliderMax=1, sortPriority=4}
  n4 = ScalarParameter() {defaultValue=0.5, desc=Blue Channel influence, group=01 Green Screen, parameterName=Blue Weight, sliderMax=1, sortPriority=5}
  n5 = MaterialFunctionCall!(
   IsExternalSource=StaticBoolParameter() {group=01 Green Screen, parameterName=IsMedia}
   Texture2D=TextureObjectParameter() {group=01 Green Screen, parameterName=Texture2D}
   TextureExternal=@n1
   KeyColor=@n2.RGB
   AlphaThreshold=ScalarParameter() {defaultValue=1, group=01 Green Screen, parameterName=Alpha Threshold, sortPriority=2}
   AlphaOffset=ScalarParameter() {desc=Good for filling interior holes, group=01 Green Screen, parameterName=Alpha Offset, sortPriority=3}
   WeightRed=@n3
   WeightBlue=@n4
   ClipBlack=ScalarParameter() {desc=Black point of the Alpha Matte, group=01 Green Screen, parameterName=Clip Black, sortPriority=6}
   ClipWhite=ScalarParameter() {defaultValue=1, desc=White point of the Alpha Matte, group=01 Green Screen, parameterName=Clip White, sortPriority=7}
   UnPremult=ScalarParameter() {group=Debug, parameterName=Unpremult, sliderMax=2, sliderMin=-1}
   ---------------------
   ErodeAlpha=StaticBoolParameter() {group=00_Global, parameterName=ErodeAlpha Active}
   KernalSize=ScalarParameter() {group=05_Erode, parameterName=KernalSize}
   NumSamples=ScalarParameter() {defaultValue=16, group=05_Erode, parameterName=Samples}
  ) {materialFunction=/Engine/Functions/Engine_MaterialFunctions01/Chromakeying/DiffColorKeyerErodeSinglePass}
  MaterialExpressionTextureCoordinate_0 = TextureCoordinate()
  MaterialExpressionScalarParameter_8 = ScalarParameter() {defaultValue=0.1, group=06_FadeEdges, parameterName=MarginX, sliderMax=0.5}
  MaterialExpressionScalarParameter_17 = ScalarParameter() {defaultValue=0.1, group=06_FadeEdges, parameterName=MarginY, sliderMax=0.5}
  MaterialExpressionScalarParameter_21 = ScalarParameter() {defaultValue=0.1, group=06_FadeEdges, parameterName=Radius, sliderMax=0.5}
  MaterialExpressionScalarParameter_22 = ScalarParameter() {defaultValue=0.1, group=06_FadeEdges, parameterName=Feather, sliderMax=1}
  n6 = Custom!(
   uv=@MaterialExpressionTextureCoordinate_0
   marginx=@MaterialExpressionScalarParameter_8
   marginy=@MaterialExpressionScalarParameter_17
   radius=@MaterialExpressionScalarParameter_21
   feather=@MaterialExpressionScalarParameter_22
  ) {
   code="float2 offset = abs(uv - 0.5) - (0.5 - float2(marginx, marginy) - radius);\r\nfloat distance = length(max(offset, 0.0)) - radius;\r\n\r\nreturn 1.0-smoothstep(0.0, feather, distance);"
   description=FadeEdges
   outputType=CMOT_Float1
   inputs [
    {inputName=uv, input {expression=&self:MaterialExpressionTextureCoordinate_0}}
    {inputName=marginx, input {expression=&self:MaterialExpressionScalarParameter_8}}
    {inputName=marginy, input {expression=&self:MaterialExpressionScalarParameter_17}}
    {inputName=radius, input {expression=&self:MaterialExpressionScalarParameter_21}}
    {inputName=feather, input {expression=&self:MaterialExpressionScalarParameter_22}}
   ]
  }
  n7 = ScalarParameter() {group=Debug, parameterName=Debug Alpha, sliderMax=1}
  LinearInterpolate(
   A=StaticSwitchParameter(True=Multiply(A=@n5.Alpha, B=@n6), False=@n5.Alpha) {group=00_Global, parameterName=FadeEdges Active}
   B=Constant() {r=1}
   Alpha=@n7
  )
  n8 = VectorParameter() {group=03_Color Grade, parameterName=Gain, defaultValue=<rgba 1 1 1 0>}
  n9 = VectorParameter() {group=03_Color Grade, parameterName=Gamma, defaultValue=<rgba 1 1 1 0>}
  n10 = VectorParameter() {group=03_Color Grade, parameterName=Lift}
  n11 = ScalarParameter() {defaultValue=1, desc=Color Grading Brightness, group=03_Color Grade, parameterName=Contrast}
  n12 = ScalarParameter() {desc=Color Grading Saturation, group=03_Color Grade, parameterName=Saturation}
  n13 = ScalarParameter() {desc=Color Grading Hue Shift, group=03_Color Grade, parameterName=Hue Shift}
  n14 = ScalarParameter() {defaultValue=1, desc=Color Grading Brightness, group=03_Color Grade, parameterName=Brightness}
  n15 = ScalarParameter() {defaultValue=0.25, desc="Despill hue range.  Based on the color selection.", group=02_Despill, parameterName=DeSpill Hue Range}
  n16 = MaterialFunctionCall!(
   Texture=@n5.RGB
   Despill Hue=@n2.RGB
   Hue Range=@n15
   DeSpill Amount=ScalarParameter() {defaultValue=1, desc="Despill Hue Saturation.  ", group=02_Despill, parameterName=DeSpill Amount, sliderMax=1}
  ) {materialFunction=/Engine/Functions/Engine_MaterialFunctions01/Chromakeying/DespillByHue}
  n17 = ScalarParameter() {defaultValue=1, desc="0 = Despill Hue 1 = Automatic Despill", group=02_Despill, parameterName=Despill Method, sliderMax=1, sortPriority=1}
  n18 = StaticSwitchParameter(
   True=LinearInterpolate(
    A=MaterialFunctionCall!(Image Color=@n16.Result) {materialFunction=/Engine/Functions/Engine_MaterialFunctions01/Chromakeying/DespillByAvg}.Result
    B=MaterialFunctionCall(Image Color=@n5.RGB) {materialFunction=&DespillByAvg}.Result
    Alpha=@n17
   )
   False=@n5.RGB
  ) {group=00_Global, parameterName=Despill Active, sortPriority=1}
  LinearInterpolate(
   A=StaticSwitchParameter(
    True=MaterialFunctionCall!(Gain=@n8.RGB, Gamma=@n9.RGB, Lift=@n10.RGB, Contrast=@n11, Saturation=@n12, Hue Shift=@n13, Brightness=@n14, Color=@n18) {materialFunction=/Engine/Functions/Engine_MaterialFunctions01/ImageAdjustment/ColorCorrection}.Result
    False=@n18
   ) {group=00_Global, parameterName=Color Grade Active, sortPriority=5}
   B=@n5.Alpha
   Alpha=@n7
  )
 ]
}
